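# Sound cue for equipping a consumable in the Creative Prop Hunt mode.
# A single wave player plays one equip sound in the consumables sound class, with a slight volume boost.
# It uses the Prop Hunt attenuation asset and overrides concurrency to cap simultaneous instances, stopping the oldest.
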SoundCue /Game/Sounds/Creative/Modes/PropHunt/PropHunt_Consumable_Equip_Cue
root=WavePlayer_1
nodes [{id=WavePlayer_1, properties {soundWaveAssetPtr=&/PropHunt_Consumable_Equip_01}}]
properties {
 attenuationSettings=/Game/Sounds/Creative/Attenuations/Fort_Creative_PropHunt_Att
 bOverrideConcurrency=true
 maxDistance=1400
 soundClassObject=&Fort_Consumables
 volumeMultiplier=1.15
 concurrencyOverrides {maxCount=4, resolutionRule=StopOldest}
}
