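# Sound cue for the Prop Hunt creative mode mouseover sound, as named.
# A single wave player plays the PropHunt_Mouseover_01 wave with no randomization or modulation.
# It uses the creative match attenuation with a very large max distance, the critical in-game UI sound class, and a slightly reduced volume.
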SoundCue /Game/Sounds/Creative/Modes/PropHunt/PropHunt_Mouseover_Cue
root=WavePlayer_1
nodes [{id=WavePlayer_1, properties {soundWaveAssetPtr=&/PropHunt_Mouseover_01}}]
properties {
 attenuationSettings=/Game/Sounds/Creative/Attenuations/Fort_Creative_Match_Att
 maxDistance=2097152
 soundClassObject=&Fort_UI_InGame_Critical
 volumeMultiplier=0.7
}
