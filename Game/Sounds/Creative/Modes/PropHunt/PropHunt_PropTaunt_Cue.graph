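# PropHunt_PropTaunt_Cue is a Sound Cue that plays a prop taunt for the Creative Prop Hunt mode.
# A weighted Random node picks one of seven taunt waves with replacement, so repeats can occur, and feeds a default Modulator at the root.
# The cue uses dedicated Prop Hunt taunt attenuation, the critical in-game UI sound class, reduced volume and a concurrency override capping simultaneous instances.
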SoundCue /Game/Sounds/Creative/Modes/PropHunt/PropHunt_PropTaunt_Cue
root=Modulator_1
nodes [
 {id=Modulator_1, children [Random_0]}
 {
  id=Random_0
  children [WavePlayer_0,WavePlayer_3,WavePlayer_4,WavePlayer_6,WavePlayer_7,WavePlayer_2,WavePlayer_8]
  properties {bRandomizeWithoutReplacement=false, weights [1,0.5,0.5,0.5,0.5,0.2,1]}
 }
 {id=WavePlayer_0, properties {soundWaveAssetPtr=&/PropHunt_PropTaunt_B01}}
 {id=WavePlayer_2, properties {soundWaveAssetPtr=&/PropHunt_PropTaunt_B10}}
 {id=WavePlayer_3, properties {soundWaveAssetPtr=&/PropHunt_PropTaunt_B09}}
 {id=WavePlayer_4, properties {soundWaveAssetPtr=&/PropHunt_PropTaunt_B08}}
 {id=WavePlayer_6, properties {soundWaveAssetPtr=&/PropHunt_PropTaunt_B06}}
 {id=WavePlayer_7, properties {soundWaveAssetPtr=&/PropHunt_PropTaunt_B05}}
 {id=WavePlayer_8, properties {soundWaveAssetPtr=&/PropHunt_PropTaunt_B04}}
]
properties {
 attenuationSettings=/Game/Sounds/Creative/Attenuations/Fort_Creative_PropHunt_Taunt_Att
 bOverrideConcurrency=true
 maxDistance=6000
 soundClassObject=&Fort_UI_InGame_Critical
 volumeMultiplier=0.35
 concurrencyOverrides {maxCount=5}
}
