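# PropHunt_Timer_Finished_Cue is a Sound Cue for the Creative Prop Hunt mode, named for the moment its timer finishes.
# A single Wave Player node, with no mixing or randomization, plays the orchestral mission-start music stinger at reduced volume.
# It uses the Creative match attenuation settings and routes through the critical in-game UI sound class.
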SoundCue /Game/Sounds/Creative/Modes/PropHunt/PropHunt_Timer_Finished_Cue
root=WavePlayer_1
nodes [{id=WavePlayer_1, properties {soundWaveAssetPtr=/Game/Sounds/Fort_Music/Orchestral_Score/Orch_Stingers/Stingers_MISSION_START_175BPM}}]
properties {
 attenuationSettings=/Game/Sounds/Creative/Attenuations/Fort_Creative_Match_Att
 maxDistance=2097152
 soundClassObject=&Fort_UI_InGame_Critical
 volumeMultiplier=0.6
}
