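# Sound cue for the Prop Hunt creative mode, named as the alert for an imminent timer.
# A single wave player feeds the PropHunt_Timer_Imminent_01 wave straight to the output, with no randomization or modulation.
# It routes to the critical in-game UI sound class and uses the shared Creative match attenuation with a very large max distance.
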SoundCue /Game/Sounds/Creative/Modes/PropHunt/PropHunt_Timer_Imminent_Cue
root=WavePlayer_0
nodes [{id=WavePlayer_0, properties {soundWaveAssetPtr=&/PropHunt_Timer_Imminent_01}}]
properties {
 attenuationSettings=/Game/Sounds/Creative/Attenuations/Fort_Creative_Match_Att
 maxDistance=2097152
 soundClassObject=&Fort_UI_InGame_Critical
 volumeMultiplier=1
}
