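# Sound cue for the Creative Prop Hunt mode, played, going by its name, when a prop turns back into a player.
# A mixer layers the main transform wave with a quieter consumable-equip wave, which is briefly delayed and shaped by a volume envelope that swells then fades.
# It uses Prop Hunt attenuation, the critical in-game UI sound class, and both global and per-owner concurrency sets.
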SoundCue /Game/Sounds/Creative/Modes/PropHunt/PropHunt_TurnIntoPlayer_Cue
root=Mixer_0
nodes [
 {id=Delay_0, children [Enveloper_0], properties {delayMax=0.3, delayMin=0.3}}
 {
  id=Enveloper_0
  children [WavePlayer_0]
  properties {pitchCurve {editorCurveData {keys [{time=0}]}}, volumeCurve {editorCurveData {keys [<at 0 0.1>,<at 0.5 1>,<at 3 0>]}}}
 }
 {id=Mixer_0, children [WavePlayer_1,Delay_0], properties {inputVolume [1,0.6]}}
 {id=WavePlayer_0, properties {soundWaveAssetPtr=&/PropHunt_Consumable_Equip_02}}
 {id=WavePlayer_1, properties {soundWaveAssetPtr=&/PropHunt_TurnIntoPlayer_01}}
]
properties {
 attenuationSettings=/Game/Sounds/Creative/Attenuations/Fort_Creative_PropHunt_Att
 maxDistance=1400
 soundClassObject=&Fort_UI_InGame_Critical
 volumeMultiplier=0.6
 concurrencyOverrides {bLimitToOwner=true, maxCount=1}
 concurrencySet [&/CONC_ProHunt_TurnInto_Global, &/CONC_ProHunt_TurnInto_Owner]
}
