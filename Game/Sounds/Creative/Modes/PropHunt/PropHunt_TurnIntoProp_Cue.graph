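# Sound cue for Creative Prop Hunt, named for the moment a player turns into a prop.
# A mixer layers two wave players at equal volume, so both waves play together rather than as random variations.
# It uses the Prop Hunt attenuation, the critical in-game UI sound class and a reduced overall volume.
# Global and per-owner concurrency sets are assigned, alongside override values for a single owner-limited instance.
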SoundCue /Game/Sounds/Creative/Modes/PropHunt/PropHunt_TurnIntoProp_Cue
root=Mixer_0
nodes [
 {id=Mixer_0, children [WavePlayer_0,WavePlayer_1], properties {inputVolume [1,1]}}
 {id=WavePlayer_0, properties {soundWaveAssetPtr=&/PropHunt_TurnIntoProp_01}}
 {id=WavePlayer_1, properties {soundWaveAssetPtr=&/PropHunt_TurnIntoProp_02}}
]
properties {
 attenuationSettings=/Game/Sounds/Creative/Attenuations/Fort_Creative_PropHunt_Att
 maxDistance=1400
 soundClassObject=&Fort_UI_InGame_Critical
 volumeMultiplier=0.65
 concurrencyOverrides {bLimitToOwner=true, maxCount=1}
 concurrencySet [&/CONC_ProHunt_TurnInto_Global, &/CONC_ProHunt_TurnInto_Owner]
}
